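# Sound cue for a dropped machine pistol, using the loot-collect sound class and weapon-drop attenuation.
# A Modulator feeds a quality switch: the first two slots pick randomly among three equally weighted suppressed auto-pistol drop waves, while the last two always play the third wave.
# The cue raises pitch, lowers volume slightly and joins the shared stop-oldest weapon-drop concurrency group.
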SoundCue /Game/Sounds/Fort_Weapons/pistol/MachinePistol/SC_Machine_Pistol_Dropped
root=Modulator_0
nodes [
 {id=Modulator_0, children [QualityLevel_0]}
 {id=QualityLevel_0, children [Random_0,Random_0,WavePlayer_5,WavePlayer_5]}
 {id=Random_0, children [WavePlayer_0,WavePlayer_4,WavePlayer_5], properties {weights [1,1,1]}}
 {id=WavePlayer_0, properties {soundWaveAssetPtr=/Game/Sounds/Fort_Weapons/pistol/MachinePistol/SoundWaves/Pistol_AutoHeavySupp_Drop_01}}
 {id=WavePlayer_4, properties {soundWaveAssetPtr=/Game/Sounds/Fort_Weapons/pistol/MachinePistol/SoundWaves/Pistol_AutoHeavySupp_Drop_02}}
 {id=WavePlayer_5, properties {soundWaveAssetPtr=/Game/Sounds/Fort_Weapons/pistol/MachinePistol/SoundWaves/Pistol_AutoHeavySupp_Drop_03}}
]
properties {
 attenuationSettings=&Fort_Weapon_Drop_Att
 maxDistance=3000
 pitchMultiplier=1.2
 soundClassObject=&Fort_Loot_Collect
 volumeMultiplier=0.8
 concurrencyOverrides {resolutionRule=PreventNew}
 concurrencySet [&Weapon_Drop_StopOldest]
}
