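# First-person fire sound cue for the machine pistol, mixing a shot layer and a low-ammo layer under separate sound class overrides.
# The shot layer blends random close-fire, mechanism and shell-impact waves, with volume continuously scaled down by a TimeSinceStarted parameter.
# A LowAmmo branch adds a pitched-up warning wave in the critical low-ammo class.
# The FirstShot branch is mixed in but has no sounds connected.
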SoundCue /Game/Sounds/Fort_Weapons/pistol/MachinePistol/SC_Machine_Pistol_Fire_Quartz_1P
root=Mixer_2
nodes [
 {id=Branch_1, children [null,null,null], properties {boolParameterName=FirstShot}}
 {id=Branch_3, children [Modulator_8,null,null], properties {boolParameterName=LowAmmo}}
 {id=Mixer_0, children [Random_0,Random_2,Modulator_2], properties {inputVolume [1,0.3,0.3]}}
 {id=Mixer_2, children [SoundClass_0,SoundClass_1], properties {inputVolume [1,1]}}
 {id=Mixer_3, children [Branch_1, ModulatorContinuous_0], properties {inputVolume [1,1]}}
 {
  id=ModulatorContinuous_0
  children [Modulator_5]
  properties {volumeModulationParams {default=0, maxOutput=0.6, minOutput=1, parameterName=TimeSinceStarted}}
 }
 {id=Modulator_0, children [Mixer_0], properties {pitchMax=1, pitchMin=1}}
 {id=Modulator_2, children [Random_5]}
 {id=Modulator_5, children [Modulator_0], properties {pitchMax=1, pitchMin=1}}
 {id=Modulator_6, children [Branch_3], properties {pitchMax=1, pitchMin=1, volumeMax=1.15, volumeMin=1}}
 {id=Modulator_8, children [WavePlayer_31], properties {pitchMax=1.15, pitchMin=1.15, volumeMax=1, volumeMin=1}}
 {
  id=Random_0
  children [
   WavePlayer_1
   WavePlayer_2
   WavePlayer_4
   WavePlayer_5
   WavePlayer_13
   WavePlayer_14
   WavePlayer_15
   WavePlayer_16
   WavePlayer_17
   WavePlayer_18
   WavePlayer_20
   WavePlayer_21
  ]
  properties {weights [1,1,1,1,1,1,1,1,1,1,1,1]}
 }
 {
  id=Random_2
  children [WavePlayer_22, WavePlayer_23, WavePlayer_24, WavePlayer_25, WavePlayer_26, WavePlayer_27]
  properties {weights [1,1,1,1,1,1]}
 }
 {
  id=Random_5
  children [WavePlayer_32, WavePlayer_33, WavePlayer_36, WavePlayer_37, WavePlayer_38, WavePlayer_39]
  properties {weights [1,1,1,1,1,1]}
 }
 {
  id=SoundClass_0
  children [Modulator_6]
  properties {soundClassOverride=/Game/Sounds/FortSoundClassesAndModes/Fort_WeaponRanged_LowAmmo_Critical_PSM}
 }
 {id=SoundClass_1, children [Mixer_3], properties {soundClassOverride=&Fort_WeaponRanged_Fire_1P_PSM}}
 {id=WavePlayer_1, properties {soundWaveAssetPtr=/Game/Sounds/Fort_Weapons/pistol/MachinePistol/SoundWaves/Machine_Pistol2_Fire_Close_10}}
 {id=WavePlayer_13, properties {soundWaveAssetPtr=/Game/Sounds/Fort_Weapons/pistol/MachinePistol/SoundWaves/Machine_Pistol2_Fire_Close_04}}
 {id=WavePlayer_14, properties {soundWaveAssetPtr=/Game/Sounds/Fort_Weapons/pistol/MachinePistol/SoundWaves/Machine_Pistol2_Fire_Close_05}}
 {id=WavePlayer_15, properties {soundWaveAssetPtr=/Game/Sounds/Fort_Weapons/pistol/MachinePistol/SoundWaves/Machine_Pistol2_Fire_Close_06}}
 {id=WavePlayer_16, properties {soundWaveAssetPtr=/Game/Sounds/Fort_Weapons/pistol/MachinePistol/SoundWaves/Machine_Pistol2_Fire_Close_07}}
 {id=WavePlayer_17, properties {soundWaveAssetPtr=/Game/Sounds/Fort_Weapons/pistol/MachinePistol/SoundWaves/Machine_Pistol2_Fire_Close_08}}
 {id=WavePlayer_18, properties {soundWaveAssetPtr=/Game/Sounds/Fort_Weapons/pistol/MachinePistol/SoundWaves/Machine_Pistol2_Fire_Close_09}}
 {id=WavePlayer_2, properties {soundWaveAssetPtr=/Game/Sounds/Fort_Weapons/pistol/MachinePistol/SoundWaves/Machine_Pistol2_Fire_Close_01}}
 {id=WavePlayer_20, properties {soundWaveAssetPtr=/Game/Sounds/Fort_Weapons/pistol/MachinePistol/SoundWaves/Machine_Pistol2_Fire_Close_11}}
 {id=WavePlayer_21, properties {soundWaveAssetPtr=/Game/Sounds/Fort_Weapons/pistol/MachinePistol/SoundWaves/Machine_Pistol2_Fire_Close_12}}
 {id=WavePlayer_22, properties {soundWaveAssetPtr=/Game/Sounds/Fort_Weapons/pistol/MachinePistol/SoundWaves/Machine_Pistol2_Fire_CloseMech_01}}
 {id=WavePlayer_23, properties {soundWaveAssetPtr=/Game/Sounds/Fort_Weapons/pistol/MachinePistol/SoundWaves/Machine_Pistol2_Fire_CloseMech_02}}
 {id=WavePlayer_24, properties {soundWaveAssetPtr=/Game/Sounds/Fort_Weapons/pistol/MachinePistol/SoundWaves/Machine_Pistol2_Fire_CloseMech_03}}
 {id=WavePlayer_25, properties {soundWaveAssetPtr=/Game/Sounds/Fort_Weapons/pistol/MachinePistol/SoundWaves/Machine_Pistol2_Fire_CloseMech_04}}
 {id=WavePlayer_26, properties {soundWaveAssetPtr=/Game/Sounds/Fort_Weapons/pistol/MachinePistol/SoundWaves/Machine_Pistol2_Fire_CloseMech_05}}
 {id=WavePlayer_27, properties {soundWaveAssetPtr=/Game/Sounds/Fort_Weapons/pistol/MachinePistol/SoundWaves/Machine_Pistol2_Fire_CloseMech_06}}
 {id=WavePlayer_31, properties {soundWaveAssetPtr=/Game/Sounds/Fort_Weapons/pistol/MachinePistol/SoundWaves/Machine_Pistol2_LowAmmo_02}}
 {id=WavePlayer_32, properties {soundWaveAssetPtr=/Game/Sounds/Fort_Weapons/pistol/MachinePistol/SoundWaves/Machine_Pistol2_ShellImpact_01}}
 {id=WavePlayer_33, properties {soundWaveAssetPtr=/Game/Sounds/Fort_Weapons/pistol/MachinePistol/SoundWaves/Machine_Pistol2_ShellImpact_02}}
 {id=WavePlayer_36, properties {soundWaveAssetPtr=/Game/Sounds/Fort_Weapons/pistol/MachinePistol/SoundWaves/Machine_Pistol2_ShellImpact_03}}
 {id=WavePlayer_37, properties {soundWaveAssetPtr=/Game/Sounds/Fort_Weapons/pistol/MachinePistol/SoundWaves/Machine_Pistol2_ShellImpact_04}}
 {id=WavePlayer_38, properties {soundWaveAssetPtr=/Game/Sounds/Fort_Weapons/pistol/MachinePistol/SoundWaves/Machine_Pistol2_ShellImpact_05}}
 {id=WavePlayer_39, properties {soundWaveAssetPtr=/Game/Sounds/Fort_Weapons/pistol/MachinePistol/SoundWaves/Machine_Pistol2_ShellImpact_06}}
 {id=WavePlayer_4, properties {soundWaveAssetPtr=/Game/Sounds/Fort_Weapons/pistol/MachinePistol/SoundWaves/Machine_Pistol2_Fire_Close_02}}
 {id=WavePlayer_5, properties {soundWaveAssetPtr=/Game/Sounds/Fort_Weapons/pistol/MachinePistol/SoundWaves/Machine_Pistol2_Fire_Close_03}}
]
properties {
 attenuationSettings=&Fort_Weapon_Projectile_FirstPerson_Athena_Att
 bDebug=true
 maxDistance=2097152
 soundClassObject=&Fort_WeaponRanged_Fire_1P_PSM
 volumeMultiplier=1
 concurrencySet [&Weapon_Gunfire_FirstPerson_Conc]
}
