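# Sound cue for picking up the machine pistol weapon.
# A single wave player feeds the machine pistol pickup wave through a modulator at reduced cue volume.
# It uses the shared weapon pickup attenuation and loot collect sound class.
# A concurrency override allows only one instance at a time, blocking new plays while one is active.
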SoundCue /Game/Sounds/Fort_Weapons/pistol/MachinePistol/SC_Machine_Pistol_Pickup
root=Modulator_0
nodes [
 {id=Modulator_0, children [WavePlayer_0]}
 {id=WavePlayer_0, properties {soundWaveAssetPtr=/Game/Sounds/Fort_Weapons/pistol/MachinePistol/SoundWaves/MachinePistol_Pickup}}
]
properties {
 attenuationSettings=&Fort_Weapon_Pickup_Att
 bOverrideConcurrency=true
 maxDistance=3500
 soundClassObject=&Fort_Loot_Collect
 volumeMultiplier=0.25
 concurrencyOverrides {maxCount=1, resolutionRule=PreventNew}
}
