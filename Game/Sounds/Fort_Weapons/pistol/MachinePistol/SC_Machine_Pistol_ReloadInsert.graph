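# Sound cue for the magazine-insert step of the machine pistol's reload foley.
# A fixed short delay node feeds a single wave player for the insert sample, so the sound lags slightly behind its trigger.
# It routes through the weapon foley sound class with reload attenuation and a slightly reduced volume, overrides concurrency, and has its debug flag switched on.
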SoundCue /Game/Sounds/Fort_Weapons/pistol/MachinePistol/SC_Machine_Pistol_ReloadInsert
root=Delay_0
nodes [
 {id=Delay_0, children [WavePlayer_1], properties {delayMax=0.15, delayMin=0.15}}
 {id=WavePlayer_1, properties {soundWaveAssetPtr=/Game/Sounds/Fort_Weapons/pistol/MachinePistol/SoundWaves/Machine_Pistol2_ReloadInsert_01}}
]
properties {
 attenuationSettings=&Fort_Weapon_Reload_Att
 bDebug=true
 bOverrideConcurrency=true
 maxDistance=2250
 soundClassObject=&Fort_WeaponFoley
 volumeMultiplier=0.752
}
